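# Niagara vehicle cosmetic drift trail that draws chain and flame ribbons, sparks and wheel meshes at the car's wheels.
# Chains spawns two socket-following ribbons; Wheel scales meshes to wheel size, and FN_DisableTireFX masks effects per wheel.
# BounceSparks_Generate emits death events alongside a BounceSparks_DeathEvent001 emitter, while NE_VC_NEW_SPARKS merges spark layers through spawn groups and vortex force.
# Painted-variant overrides color each layer, with age-based gradients on the chains.
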NiagaraSystem /VehicleCosmetics/DriftTrail/DriftTrail_Chains/FX/NS_DriftTrail_Chains
dependencies {
 usedDataInterfaces [
  &NiagaraDataInterfaceCurve
  &NiagaraDataInterfaceColorCurve
  &NiagaraDataInterfaceArrayColor
  &NiagaraDataInterfaceVectorCurve
  &NiagaraDataInterfaceSkeletalMesh
 ]
 usedDynamicInputs [&VC_Mask_Vectors]
 usedModules [
  &self:SystemSpawnScript.SystemScriptSource.SystemScriptGraph.NiagaraNodeAssignment_0.SetVariables_#1
  &SystemState
  &FN_UnpackVehicleFlags
  /Game/Effects/Niagara/Modules/Vehicle/FN_UpdateVehicleAttribs
  /VehicleCosmetics/FX/Niagara/Modules/VC_SetVehicleSpeedAttribs
  /Game/Effects/Niagara/Modules/Vehicle/FN_VehicleFlags_Cheatcodes
 ]
}
emitters {
 BounceSparks_DeathEvent001 {
  emitterUpdateScript [
   &EmitterState(Life Cycle Mode=::NewEnumerator1, Loop Duration=0, UseLoopDelay=false)
   &SpawnRate(SpawnRate=&RandomRangeFloat(Minimum=0, Maximum=0), Use Spawn Probability=false) {enabled=false}
   &SpawnBurst_Instantaneous(Spawn Count=0, Spawn Group=1, Use Spawn Probability=false, UseLoopCountLimit=false)
  ]
  particleSpawnScript [
   &VC_PaintedVariantsOverride(
    Black=<rgba 0.01 0.01 0.01 150>
    BurntSienna=<rgba 3 0.5166 0 0.2>
    Cobalt=<rgba 0 0.185 10 1>
    Crimson=<rgba 30 0 0.8131 1>
    DefaultColor=<rgba 10 0.8333 0 1>
    ForestGreen=<rgba 0 30 1.5158 1>
    Grey=<rgba 5 5 5 1>
    Lime=<rgba 5 10 0 1>
    Pink=<rgba 10 1.1979 6.7992 1>
    Purple=<rgba 4.1041 0 10 1>
    Saffron=<rgba 10 7.6667 0 1>
    SkyBlue=<rgba 0 6.6539 10 1>
    TitaniumWhite=<rgba 10 10 10 1>
   )
   /SetVariables(Particles.RenderVisTag:NiagaraInt32=&SpawnGroup()) {isSetParameters=true}
   &InitializeParticle(
    Color Mode=::NewEnumerator0
    Lifetime=&RandomRangeFloat(Minimum=0.2, Maximum=0.2)
    Sprite Rotation Mode=::NewEnumerator2
    Sprite Size Max=<0.1 5>
    Sprite Size Min=<0.5 1>
    Sprite Size Mode=::NewEnumerator2
    UsePositionOffset=false
    Write Material Random=true
   )
   &PartitionParticles()
   &SkeletalMeshLocation(
    Bone / Socket Index=&Add_Integer(A=$Particles.PartitionParticles.Partition, B=4)
    Bone / Socket Sampling Mode=::NewEnumerator5
    Interpolate Mesh Sampling=false
    MaskSpawnGroup=false
    Mesh Sampling Type=::NewEnumerator6
    Sampled Position Offset=&RandomRangeVector(
     Minimum=&RandomRangeVector(Minimum=<-50 0 0>, Maximum=<-80 1 1>)
     Maximum=&RandomRangeVector(Minimum=<-800 0 0>, Maximum=<-500 1 1>)
    )
    Skeletal Mesh=$System.SkeletalMesh
   ) {enabled=false}
   &AddVelocity(Module.Use Falloff Distance=false, Override Seed=false, Velocity Mode=::NewEnumerator1, Velocity Speed=-250) {enabled=false}
   &ShapeLocation(MaskSpawnGroup=false, Override Seed=false, Sphere Radius=150, UseOwnerScale=true)
  ]
  particleUpdateScript [
   &FN_DisableTireFX(
    Mask Index Attribute:NiagaraInt32=$Particles.PartitionParticles.Partition
    Mask Index FL=2
    Mask Index FR=3
    Mask Index RL=0
    Mask Index RR=1
   )
   &ParticleState()
   &ScaleSpriteSize(Uniform Curve Sprite Scale=DataInterface() {propertyValues {Curve {keys [<at 0 2 RCIM_Cubic>,<at 1 0.5 RCIM_Cubic>]}}})
   &SolveForcesAndVelocity(Clamp Velocity=false, Limit Acceleration=false)
   &ScaleColor(
    Scale Alpha=2
    Scale RGB=&VectorFromFloat(Value=&FloatFromCurve(
      FloatCurve=DataInterface() {propertyValues {Curve {
         defaultValue=0
         keys [
          {arriveTangent=9.0725, interpMode=RCIM_Cubic, leaveTangent=9.0725, tangentMode=RCTM_User, time=0, value=0}
          {arriveTangent=-2.5249, interpMode=RCIM_Cubic, leaveTangent=-2.5249, tangentMode=RCTM_User, time=0.1408, value=5}
          {arriveTangent=-0.0405, interpMode=RCIM_Cubic, leaveTangent=-0.0405, tangentMode=RCTM_User, time=1, value=0}
         ]
        }}}
      Scale Curve=2
     ))
    ScaleA=true
    ScaleRGB=true
   )
  ]
  rendererClasses [&NiagaraSpriteRendererProperties]
  renderers [
   {rendererClass=&NiagaraSpriteRendererProperties}
   {rendererClass=&NiagaraSpriteRendererProperties, rendererIndex=1}
  ]
 }
 BounceSparks_Generate {
  emitterUpdateScript [
   &EmitterState(Inactive Response=::NewEnumerator1, Life Cycle Mode=::NewEnumerator1, Loop Duration=1, UseLoopDelay=false)
   &SpawnBurst_Instantaneous(Spawn Count=4, Use Spawn Probability=false, UseLoopCountLimit=false)
   &SpawnRate(SpawnRate=5, Use Spawn Probability=false) {enabled=false}
  ]
  particleSpawnScript [
   &VC_PaintedVariantsOverride(Black=<rgba 0 0 0 1>, BurntSienna=<rgba 0.0642 0.0136 0 1>, DefaultColor=<rgba 1 0.1328 0 1>)
   &InitializeParticle(
    Color Mode=::NewEnumerator0
    Lifetime Max=0.05
    Lifetime Min=0.01
    Lifetime Mode=::NewEnumerator1
    Sprite Rotation Mode=::NewEnumerator2
    Sprite Size Mode=::NewEnumerator3
    Uniform Sprite Size=60
    UsePositionOffset=false
    Write Material Random=true
   )
   &PartitionParticles()
   &SkeletalMeshLocation(
    Bone / Socket Index=&Add_Integer(A=$Particles.PartitionParticles.Partition, B=4)
    Bone / Socket Sampling Mode=::NewEnumerator5
    Interpolate Mesh Sampling=false
    MaskSpawnGroup=false
    Mesh Sampling Type=::NewEnumerator6
    Sampled Position Offset=&AddVectorToPosition(Position=<-360 0 -20>, Vector=$User.offset)
    Skeletal Mesh=$System.SkeletalMesh
   )
   &AddVelocity(Velocity=<-1500 0 0>) {enabled=false}
  ]
  particleUpdateScript [
   &FN_DisableTireFX(
    Mask Index Attribute:NiagaraInt32=$Particles.PartitionParticles.Partition
    Mask Index FL=2
    Mask Index FR=3
    Mask Index RL=0
    Mask Index RR=1
   )
   &ParticleState()
   &ScaleSpriteSize(Uniform Curve Sprite Scale=DataInterface() {propertyValues {Curve {defaultValue=0, keys [{tangentMode=RCTM_User, time=0, value=1}, {tangentMode=RCTM_User, time=1, value=0}]}}})
   &SolveForcesAndVelocity(Clamp Velocity=false, Limit Acceleration=false)
   &GenerateDeathEvent(Use Event Probability=false)
   &ScaleColor(
    Scale Alpha=&FloatFromCurve(FloatCurve=DataInterface() {propertyValues {Curve {keys [<at 0 1>,<at 1 0>]}}})
    Scale RGB=&VectorFromFloat(Value=&FloatFromCurve(FloatCurve=DataInterface() {propertyValues {Curve {keys [<at 0 1>,<at 1 0>]}}}, Scale Curve=50))
    ScaleA=true
    ScaleRGB=true
   )
  ]
  rendererClasses [&NiagaraSpriteRendererProperties]
  renderers [{rendererClass=&NiagaraSpriteRendererProperties}]
 }
 Chains {
  emitterUpdateScript [
   &EmitterState(Inactive Response=::NewEnumerator1, Life Cycle Mode=::NewEnumerator1, Loop Duration=0, UseLoopDelay=false)
   &VC_RateQualityHandler(High Spawn Rate=60, Low Spawn Rate=60, Medium Spawn Rate=60, Use Custom Rate?=true)
   &VC_SpawnRibbon_v1(Number of Ribbons=2, SpawnRate=90)
  ]
  particleSpawnScript [
   &VC_PaintedVariantsOverride(
    Black=<rgba 2 2 2 1>
    BurntSienna=<rgba 80 13.3816 0 1>
    Cobalt=<rgba 0 27.7487 1500 1>
    Crimson=<rgba 350 0.0002 0 1>
    DefaultColor=<rgba 450 84.9433 0 1>
    ForestGreen=<rgba 0 450 22.7371 1>
    Grey=<rgba 30 30 30 1>
    Lime=<rgba 124.9999 250 0 1>
    Pink=<rgba 500 59.8961 339.9619 1>
    Purple=<rgba 225.7278 0 550 1>
    R1 Black:LinearColor=<rgba 0.3 0.3 0.3 1>
    R1 BurntSienna:LinearColor=<rgba 1 0.2116 0 1>
    R1 Cobalt:LinearColor=<rgba 0 0.925 50 1>
    R1 Crimson:LinearColor=<rgba 150 0.0001 0 0.2>
    R1 DefaultColor=<rgba 20 1.6667 0 1>
    R1 Grey:LinearColor=<rgba 1 1 1 1>
    R1 Lime:LinearColor=<rgba 10 20 0 1>
    R1 Pink:LinearColor=<rgba 10 1.1979 6.7992 1>
    R1 Purple:LinearColor=<rgba 8.2083 0 20 1>
    R1 Saffron:LinearColor=<rgba 10 7.6667 0 1>
    R1 SkyBlue:LinearColor=<rgba 0 13.3077 20 1>
    R1 TitaniumWhite:LinearColor=<rgba 150 150 150 1>
    Saffron=<rgba 550 421.6667 0 1>
    SkyBlue=<rgba 0 549.9999 550 1>
    TitaniumWhite=<rgba 550 550 550 1>
    Use R1 Painted Overrides=true
   ) {enabled=false}
   &VC_InitializeRibbon_v1(
    Color=$Particles.Color
    Custom Offset=true
    Enable Skeletal Position=true
    Facing Axis:ENiagaraOrientationAxis=::YAxis
    Local Vector Offset=&AddVectorToPosition(Position=<-30 0 -25>, Vector=$User.offset)
    Offset Amount=1
    Ribbon Width=50
   )
   &AddVelocity(Velocity=<-1600 0 50>) {module=AddVelocity001}
   /SetVariables() {isSetParameters=true}
  ]
  particleUpdateScript [
   &FN_DisableTireFX(Mask Index Attribute:NiagaraInt32=$Particles.SpawnGroup, Mask Index FL=2, Mask Index FR=3, Mask Index RL=0, Mask Index RR=1)
   &InheritSourceMovement(
    Applied Position Scale=&VectorFromFloat(Value=&FloatFromCurve(FloatCurve=DataInterface() {propertyValues {Curve {keys [<at 0.3 1 RCIM_Cubic>, {tangentMode=RCTM_User, time=1, value=0}]}}}))
    ApplyForce=false
    ApplyPosition=true
    ApplyPositionFalloff=false
   )
   &ParticleState()
   &SolveForcesAndVelocity(Clamp Velocity=false, Limit Acceleration=false)
   &VC_PaintedVariantsOverride(
    Black=&ColorFromCurve(ColorCurve=DataInterface() {propertyValues {
       AlphaCurve {keys [<at 0 1 RCIM_Cubic>,<at 1 1 RCIM_Cubic>]}
       BlueCurve {keys [<at 0 0.025 RCIM_Cubic>,<at 0.7088 0.15>,<at 1 0.2>]}
       GreenCurve {keys [<at 0 0.025 RCIM_Cubic>,<at 0.7088 0.15>,<at 1 0.2>]}
       RedCurve {keys [<at 0 0.025 RCIM_Cubic>,<at 0.7088 0.15>,<at 1 0.2>]}
      }})
    BurntSienna=&ColorFromCurve(ColorCurve=DataInterface() {propertyValues {
       AlphaCurve {keys [<at 0 1 RCIM_Cubic>,<at 1 1 RCIM_Cubic>]}
       BlueCurve {keys [<at 0.0675 0 RCIM_Cubic>,<at 0.3646 0.025>,<at 0.8644 0.4 RCIM_Cubic>]}
       GreenCurve {keys [<at 0.0675 0.0417 RCIM_Cubic>,<at 0.3646 0.0646>,<at 0.8644 0.4 RCIM_Cubic>]}
       RedCurve {keys [<at 0.0675 0.25 RCIM_Cubic>,<at 0.3646 0.5>,<at 0.8644 0.5 RCIM_Cubic>]}
      }})
    Cobalt=&ColorFromCurve(ColorCurve=DataInterface() {propertyValues {
       AlphaCurve {keys [<at 0 1 RCIM_Cubic>,<at 1 1 RCIM_Cubic>]}
       BlueCurve {keys [<at 0.0116 50>,<at 0.3943 10>,<at 0.5778 2>,<at 0.7473 0.75 RCIM_Cubic>]}
       GreenCurve {keys [<at 0.0116 1.9175>,<at 0.3943 0.3835>,<at 0.5778 0.1733>,<at 0.7473 0.3937 RCIM_Cubic>]}
       RedCurve {keys [<at 0.0116 0>,<at 0.3943 0>,<at 0.5778 0.1>,<at 0.7473 0.375 RCIM_Cubic>]}
      }})
    Crimson=&ColorFromCurve(ColorCurve=DataInterface() {propertyValues {
       AlphaCurve {keys [<at 0 1 RCIM_Cubic>,<at 1 1 RCIM_Cubic>]}
       BlueCurve {keys [<at 0 0.1 RCIM_Cubic>,<at 0.272 0.0315>,<at 0.576 0.075>,<at 0.7403 0.27 RCIM_Cubic>]}
       GreenCurve {keys [<at 0 0.1 RCIM_Cubic>,<at 0.272 0.0315>,<at 0.576 0.075>,<at 0.7403 0.27 RCIM_Cubic>]}
       RedCurve {keys [<at 0 20 RCIM_Cubic>,<at 0.272 5>,<at 0.576 1.5>,<at 0.7403 0.45 RCIM_Cubic>]}
      }})
    DefaultColor=&ColorFromCurve(ColorCurve=DataInterface() {propertyValues {
       AlphaCurve {keys [<at 0 1 RCIM_Cubic>,<at 1 1 RCIM_Cubic>]}
       BlueCurve {keys [<at 0 0 RCIM_Cubic>,<at 0.3471 0>,<at 0.7473 0.3>]}
       GreenCurve {keys [<at 0 0.5 RCIM_Cubic>,<at 0.3471 0.25>,<at 0.7473 0.3082>]}
       RedCurve {keys [<at 0 10 RCIM_Cubic>,<at 0.3471 5>,<at 0.7473 0.5>]}
      }})
    ForestGreen=&ColorFromCurve(ColorCurve=DataInterface() {propertyValues {
       AlphaCurve {keys [<at 0 1 RCIM_Cubic>,<at 1 1 RCIM_Cubic>]}
       BlueCurve {keys [<at 0.0063 0.0667>,<at 0.3978 0.1167>,<at 0.7263 0.3 RCIM_Cubic>]}
       GreenCurve {keys [<at 0.0063 2>,<at 0.3978 1>,<at 0.7263 0.5 RCIM_Cubic>]}
       RedCurve {keys [<at 0.0063 0>,<at 0.3978 0>,<at 0.7263 0.3333 RCIM_Cubic>]}
      }})
    Grey=&ColorFromCurve(ColorCurve=DataInterface() {propertyValues {
       AlphaCurve {keys [<at 0 1 RCIM_Cubic>,<at 1 1 RCIM_Cubic>]}
       BlueCurve {keys [<at 0 0.1 RCIM_Cubic>,<at 0.964 0.5 RCIM_Cubic>]}
       GreenCurve {keys [<at 0 0.1 RCIM_Cubic>,<at 0.964 0.5 RCIM_Cubic>]}
       RedCurve {keys [<at 0 0.1 RCIM_Cubic>,<at 0.964 0.5 RCIM_Cubic>]}
      }})
    Lime=&ColorFromCurve(ColorCurve=DataInterface() {propertyValues {
       AlphaCurve {keys [<at 0 1 RCIM_Cubic>,<at 1 1 RCIM_Cubic>]}
       BlueCurve {keys [<at 0 0 RCIM_Cubic>,<at 0.6896 0.27 RCIM_Cubic>]}
       GreenCurve {keys [<at 0 5 RCIM_Cubic>,<at 0.6896 0.45 RCIM_Cubic>]}
       RedCurve {keys [<at 0 1.6327 RCIM_Cubic>,<at 0.6896 0.3 RCIM_Cubic>]}
      }})
    Pink=&ColorFromCurve(ColorCurve=DataInterface() {propertyValues {
       AlphaCurve {keys [<at 0 1 RCIM_Cubic>,<at 1 1 RCIM_Cubic>]}
       BlueCurve {keys [<at 0 5 RCIM_Cubic>,<at 0.7508 0.4>]}
       GreenCurve {keys [<at 0 0.5 RCIM_Cubic>,<at 0.7508 0.4>]}
       RedCurve {keys [<at 0 5 RCIM_Cubic>,<at 0.7508 0.4>]}
      }})
    Purple=&ColorFromCurve(ColorCurve=DataInterface() {propertyValues {
       AlphaCurve {keys [<at 0 1 RCIM_Cubic>,<at 1 1 RCIM_Cubic>]}
       BlueCurve {keys [<at 0.2038 5 RCIM_Cubic>,<at 0.417 2>,<at 0.7805 0.5 RCIM_Cubic>]}
       GreenCurve {keys [<at 0.2038 0.2 RCIM_Cubic>,<at 0.417 0.1>,<at 0.7805 0.3 RCIM_Cubic>]}
       RedCurve {keys [<at 0.2038 1 RCIM_Cubic>,<at 0.417 0.4167>,<at 0.7805 0.3333 RCIM_Cubic>]}
      }})
    R1 Black:LinearColor=<rgba 0.001 0.001 0.001 8>
    R1 BurntSienna:LinearColor=<rgba 0.04 0.0037 0.0004 1>
    R1 Cobalt:LinearColor=<rgba 0 0.0167 0.5 2>
    R1 DefaultColor=<rgba 1 0.1333 0 1>
    R1 ForestGreen:LinearColor=<rgba 0 0.07 0 4>
    R1 Grey:LinearColor=<rgba 0.01 0.01 0.01 5>
    R1 Lime:LinearColor=<rgba 0.0287 0.1 0.005 4>
    R1 Pink:LinearColor=<rgba 0.2383 0.0157 0.25 2>
    R1 Purple:LinearColor=<rgba 0.0125 0 0.05 5>
    R1 Saffron:LinearColor=<rgba 0.1 0.0833 0 4>
    R1 TitaniumWhite:LinearColor=<rgba 0.75 0.75 0.75 0.5>
    Saffron=&ColorFromCurve(ColorCurve=DataInterface() {propertyValues {
       AlphaCurve {keys [<at 0 1 RCIM_Cubic>,<at 1 1 RCIM_Cubic>]}
       BlueCurve {keys [<at 0 0.08>,<at 0.7263 0.5 RCIM_Cubic>]}
       GreenCurve {keys [<at 0 2.9117>,<at 0.7263 0.5 RCIM_Cubic>]}
       RedCurve {keys [<at 0 4>,<at 0.7263 0.5 RCIM_Cubic>]}
      }})
    SkyBlue=&ColorFromCurve(ColorCurve=DataInterface() {propertyValues {
       AlphaCurve {keys [<at 0 1 RCIM_Cubic>,<at 1 1 RCIM_Cubic>]}
       BlueCurve {keys [<at 0 10 RCIM_Cubic>,<at 0.6949 0.3167 RCIM_Cubic>]}
       GreenCurve {keys [<at 0 5.8333 RCIM_Cubic>,<at 0.6949 0.5 RCIM_Cubic>]}
       RedCurve {keys [<at 0 0 RCIM_Cubic>,<at 0.6949 0.3 RCIM_Cubic>]}
      }})
    TitaniumWhite=&ColorFromCurve(ColorCurve=DataInterface() {propertyValues {
       AlphaCurve {keys [<at 0 1 RCIM_Cubic>,<at 1 1 RCIM_Cubic>]}
       BlueCurve {keys [<at 0 4>,<at 1 0.75 RCIM_Cubic>]}
       GreenCurve {keys [<at 0 4>,<at 1 0.75 RCIM_Cubic>]}
       RedCurve {keys [<at 0 4>,<at 1 0.75 RCIM_Cubic>]}
      }})
    Use R1 Painted Overrides=true
   ) {module=VC_PaintedVariantsOverride003}
   /SetVariables(
    Particles.ChainsScaleColor:LinearColor=&VC_DynamicScaleColor(
     Color to Scale:LinearColor=$Particles.Color
     RGB Scale:Vector3f=&MakeCustomVectorFromBool(
      Bool=$User.bIsPainted
      True Vector=&VectorFromFloat(Value=&FloatFromCurve(
        FloatCurve=DataInterface() {propertyValues {Curve {defaultValue=120, keys [{tangentMode=RCTM_User, time=0, value=1}, {tangentMode=RCTM_User, time=0.5, value=0.2}, <at 1 0.1>]}}}
        Scale Curve=150
       ))
      False Vector=&VectorFromFloat(Value=&FloatFromCurve(
        FloatCurve=DataInterface() {propertyValues {Curve {defaultValue=120, keys [{tangentMode=RCTM_User, time=0, value=12}, {tangentMode=RCTM_User, time=0.4, value=0.5}, <at 1 0.5>]}}}
        Scale Curve=70
       ))
     )
     Alpha Scale:NiagaraFloat=1
    )
    Particles.FLAMES_COLOR:LinearColor=&VC_DynamicScaleColor(
     Color to Scale:LinearColor=$Particles.R1 Color
     RGB Scale:Vector3f=&VectorFromFloat(Value=&FloatFromCurve(
       FloatCurve=DataInterface() {propertyValues {Curve {keys [<at 0 5 RCIM_Cubic>, {arriveTangent=-5, interpMode=RCIM_Cubic, leaveTangent=-5, time=0.2, value=1.75}, <at 1 0 RCIM_Cubic>]}}}
       Scale Curve=50
      ))
     Alpha Scale:NiagaraFloat=&FloatFromCurve(FloatCurve=DataInterface() {propertyValues {Curve {keys [<at 0.1 1 RCIM_Cubic>,<at 1 0 RCIM_Cubic>]}}}, Scale Curve=1.5)
    )
    Particles.ScaleRibbonWidthFlames:NiagaraFloat=&FloatFromCurve(
     FloatCurve=DataInterface() {propertyValues {Curve {keys [
         {arriveTangent=-2.5936, interpMode=RCIM_Cubic, leaveTangent=-2.5936, tangentMode=RCTM_User, time=0, value=1.15}
         {arriveTangent=-0.1454, interpMode=RCIM_Cubic, leaveTangent=-0.1454, tangentMode=RCTM_User, time=0.1, value=0.8}
         <at 1 0.5 RCIM_Cubic>
        ]}}}
     Scale Curve=52
    )
    Particles.ScaleRiibonWidthChains=80
   ) {isSetParameters=true}
   &VC_UpdateRibbon_v1(Enable Skeletal Position=true)
   &DynamicMaterialParameters(
    Index 0 Param 3=0.2
    Module.Index0Float0Write=true
    Module.Index0Float1Write=true
    Module.Index0Float2Write=true
    Module.Index0Float3Write=true
   )
  ]
  rendererClasses [&NiagaraRibbonRendererProperties]
  renderers [
   {rendererClass=&NiagaraRibbonRendererProperties}
   {rendererClass=&NiagaraRibbonRendererProperties, rendererIndex=1}
   {rendererClass=&NiagaraRibbonRendererProperties, rendererIndex=2}
  ]
 }
 NE_VC_NEW_SPARKS {
  emitterUpdateScript [
   &EmitterState(Inactive Response=::NewEnumerator1, Life Cycle Mode=::NewEnumerator1, Loop Duration=0.1, UseLoopDelay=false)
   &SpawnBurst_Instantaneous(Spawn Count=70, Spawn Group=1, Use Spawn Probability=false, UseLoopCountLimit=false)
   &SpawnRate(SpawnGroup=2, SpawnRate=120, Use Spawn Probability=false) {module=SpawnRate003}
   &SpawnRate(
    SpawnRate=/VehicleCosmetics/FX/Niagara/Modules/VC_Mask_Floats(
     FrontEnd Value:NiagaraFloat=200
     blsFrontEndPreview:NiagaraBool=false
     SpeedMin Value:NiagaraFloat=120
     SpeedMax Value:NiagaraFloat=180
     Speed Alpha:NiagaraFloat=$System.Speed_Alpha
    )
    Use Spawn Probability=false
   ) {enabled=false}
  ]
  particleSpawnScript [
   &VC_PaintedVariantsOverride(
    BurntSienna=&SelectLinearColorFromArray(
     Array Sampling Mode=::NewEnumerator0
     Direct Array Index=?
     Color Selection Array=DataInterface() {propertyValues {ColorData [<rgba 0 0 0 50>, <rgba 0.01 0.0013 0 30>]}}
     Random Seed=0
    )
    Cobalt=<rgba 0 0 0.5 10>
    Crimson=<rgba 2 0 0.0542 1>
    DefaultColor=<rgba 10 0.6667 0 2>
    ForestGreen=<rgba 0 4 0.2021 1>
    Grey=<rgba 0.015 0.015 0.015 10>
    Lime=<rgba 0.2 0.8 0 1>
    Pink=<rgba 1 0.1198 0.6799 1>
    Purple=<rgba 0.4104 0 1 1>
    R1 BurntSienna:LinearColor=<rgba 0.0642 0.0054 0 1>
    R2 Black:LinearColor=<rgba 0.001 0.001 0.001 0.5>
    R2 BurntSienna:LinearColor=<rgba 0.01 0.002 0 1>
    R2 Cobalt:LinearColor=<rgba 0 0 0.3 1>
    R2 Crimson:LinearColor=<rgba 0.2 0 0.0054 2>
    R2 DefaultColor:LinearColor=<rgba 1.5 0 0 1>
    R2 ForestGreen:LinearColor=<rgba 0 0.1 0.0051 1>
    R2 Grey:LinearColor=<rgba 0.005 0.005 0.005 1>
    R2 Lime:LinearColor=<rgba 0.0083 0.1 0 2>
    R2 Pink:LinearColor=<rgba 0.505 0 0.3214 1>
    R2 Purple:LinearColor=<rgba 0.041 0 0.1 2>
    R2 Saffron:LinearColor=<rgba 0.17 0.1303 0 2>
    R2 SkyBlue:LinearColor=<rgba 0 0.0625 0.15 1>
    R2 TitaniumWhite:LinearColor=<rgba 0.15 0.15 0.15 1>
    Saffron=<rgba 1 0.7667 0 1>
    SkyBlue=<rgba 0 0.6654 1 1>
    Use R1 Painted Overrides=true
    Use R2 Painted Overrides=true
   )
   /SetVariables(Particles.RenderTag:NiagaraInt32=&SpawnGroup()) {isSetParameters=true}
   &InitializeParticle(
    Color Mode=::NewEnumerator0
    Lifetime=/VehicleCosmetics/FX/Niagara/Modules/MergeEmitters/VC_FloatBySpawnGroup(
     SpawnGroup or ExecIndex:NiagaraInt32=0
     CustomInt:NiagaraInt32=0
     MaximumSpawnGroups:NiagaraInt32=3
     Float SG 0:NiagaraFloat=0
     Float SG 1:NiagaraFloat=&RandomRangeFloat(Minimum=0.3, Maximum=0.35)
     Float SG 2:NiagaraFloat=&RandomRangeFloat(Minimum=0.5, Maximum=1)
     Float SG 3:NiagaraFloat=0
     Float SG 4:NiagaraFloat=0
     Float SG 5:NiagaraFloat=0
    )
    Sprite Rotation Mode=::NewEnumerator2
    Sprite Size Max=<1 5>
    Sprite Size Min=<1 5>
    Sprite Size Mode=::NewEnumerator2
    UsePositionOffset=false
    Write Material Random=true
   )
   &PartitionParticles()
   &SkeletalMeshLocation(
    Bone / Socket Index=&Add_Integer(A=$Particles.PartitionParticles.Partition, B=4)
    Bone / Socket Sampling Mode=::NewEnumerator5
    Interpolate Mesh Sampling=false
    MaskSpawnGroup=false
    Mesh Sampling Type=::NewEnumerator6
    Sampled Position Offset=&AddVectorToPosition(
     Position=/VehicleCosmetics/FX/Niagara/Modules/MergeEmitters/VC_VectorBySpawnGroup(
      SpawnGroup or ExecIndex:NiagaraInt32=0
      CustomInt:NiagaraInt32=0
      MaximumSpawnGroups:NiagaraInt32=3
      Vector Group 0:Vector3f=<-50 0 -25>
      Vector Group 1:Vector3f=<-55 0 -25>
      Vector Group 2:Vector3f=<-48 0 -25>
      Vector Group 3:Vector3f=<0 0 0>
      Vector Group 4:Vector3f=<0 0 0>
      Vector Group 5:Vector3f=<0 0 0>
     )
     Vector=$User.offset
    )
    Skeletal Mesh=$System.SkeletalMesh
   )
   &AddVelocity(
    Cone Angle=&VC_FloatBySpawnGroup(
     SpawnGroup or ExecIndex:NiagaraInt32=0
     CustomInt:NiagaraInt32=0
     MaximumSpawnGroups:NiagaraInt32=3
     Float SG 0:NiagaraFloat=20
     Float SG 1:NiagaraFloat=40
     Float SG 2:NiagaraFloat=0
     Float SG 3:NiagaraFloat=0
     Float SG 4:NiagaraFloat=0
     Float SG 5:NiagaraFloat=0
    )
    Inner Cone Angle:NiagaraFloat=&VC_FloatBySpawnGroup(
     SpawnGroup or ExecIndex:NiagaraInt32=0
     CustomInt:NiagaraInt32=0
     MaximumSpawnGroups:NiagaraInt32=3
     Float SG 0:NiagaraFloat=50
     Float SG 1:NiagaraFloat=40
     Float SG 2:NiagaraFloat=0
     Float SG 3:NiagaraFloat=0
     Float SG 4:NiagaraFloat=0
     Float SG 5:NiagaraFloat=0
    )
    Module.UseConeFalloff=false
    Module.UseVelDistribution=false
    Override Seed=false
    Velocity Mode=::NewEnumerator2
    Velocity Speed=&VC_FloatBySpawnGroup(
     SpawnGroup or ExecIndex:NiagaraInt32=0
     CustomInt:NiagaraInt32=0
     MaximumSpawnGroups:NiagaraInt32=3
     Float SG 0:NiagaraFloat=0
     Float SG 1:NiagaraFloat=-150
     Float SG 2:NiagaraFloat=&SelectFloatFromArray(
      Array Sampling Mode=::NewEnumerator0
      Direct Array Index=?
      Float Selection Array=DataInterface() {propertyValues {FloatData [-70,-20]}}
      Random Seed=0
     )
     Float SG 3:NiagaraFloat=0
     Float SG 4:NiagaraFloat=0
     Float SG 5:NiagaraFloat=0
    )
   )
   &ShapeLocation(
    MaskSpawnGroup=false
    Non Uniform Scale=&VC_VectorBySpawnGroup(
     SpawnGroup or ExecIndex:NiagaraInt32=0
     CustomInt:NiagaraInt32=0
     MaximumSpawnGroups:NiagaraInt32=3
     Vector Group 0:Vector3f=<0 0 0>
     Vector Group 1:Vector3f=<1 1 1>
     Vector Group 2:Vector3f=<0 0 0>
     Vector Group 3:Vector3f=<0 0 0>
     Vector Group 4:Vector3f=<0 0 0>
     Vector Group 5:Vector3f=<0 0 0>
    )
    Override Seed=false
    Rotation Coordinate Space=::Simulation
    Sphere Radius=&VC_FloatBySpawnGroup(
     SpawnGroup or ExecIndex:NiagaraInt32=0
     CustomInt:NiagaraInt32=0
     MaximumSpawnGroups:NiagaraInt32=3
     Float SG 0:NiagaraFloat=0
     Float SG 1:NiagaraFloat=30
     Float SG 2:NiagaraFloat=0
     Float SG 3:NiagaraFloat=0
     Float SG 4:NiagaraFloat=0
     Float SG 5:NiagaraFloat=0
    )
    Sphere Surface Distribution=0.35
    UseOwnerScale=true
   )
  ]
  particleUpdateScript [
   &FN_DisableTireFX(
    Mask Index Attribute:NiagaraInt32=$Particles.PartitionParticles.Partition
    Mask Index FL=2
    Mask Index FR=3
    Mask Index RL=0
    Mask Index RR=1
   )
   &ParticleState()
   &ScaleVelocity(Velocity Scale=&VectorFromCurve(
     VectorCurve=DataInterface() {propertyValues {
       XCurve {keys [<at 0 1>,<at 1 1>]}
       YCurve {keys [<at 0 1 RCIM_Cubic>, {arriveTangent=1.0523, interpMode=RCIM_Cubic, leaveTangent=1.0523, tangentMode=RCTM_User, time=0.8, value=5}]}
       ZCurve {keys [<at 0 1 RCIM_Cubic>, {arriveTangent=1.0523, interpMode=RCIM_Cubic, leaveTangent=1.0523, tangentMode=RCTM_User, time=0.8, value=5}]}
      }}
     Scale Curve=&RandomRangeVector(Minimum=<2 0.5 0.5>, Maximum=<2 1 1>)
    ))
   &Drag(
    Drag=&VC_FloatBySpawnGroup(
     SpawnGroup or ExecIndex:NiagaraInt32=0
     CustomInt:NiagaraInt32=0
     MaximumSpawnGroups:NiagaraInt32=3
     Float SG 0:NiagaraFloat=0
     Float SG 1:NiagaraFloat=35
     Float SG 2:NiagaraFloat=0
     Float SG 3:NiagaraFloat=0
     Float SG 4:NiagaraFloat=0
     Float SG 5:NiagaraFloat=0
    )
    Use Linear Drag=true
    Use Rotational Drag=true
   )
   &VortexForce(
    Origin Pull Amount=-1500
    UseOriginOffset=true
    Vortex Axis=&RandomRangeVector(Minimum=<-1 0 0>, Maximum=<1 0 0>, Evaluation Type=::NewEnumerator1)
    Vortex Axis Coordinate Space:ENiagaraCoordinateSpace=::Local
    Vortex Force Amount=5000
    Vortex Origin:NiagaraPosition=$Particles.Initial.SkeletalMeshLocation.SampledPosition
    Vortex Origin Offset:Vector3f=<-100 0 -25>
   )
   &SolveForcesAndVelocity(Clamp Velocity=false, Limit Acceleration=false)
   /SetVariables(
    Particles.EMBERS_UP_Color:LinearColor=&VC_DynamicScaleColor(
     Color to Scale:LinearColor=$Particles.Color
     RGB Scale:Vector3f=&VectorFromFloat(Value=&FloatFromCurve(
       FloatCurve=DataInterface() {propertyValues {Curve {keys [{arriveTangent=-2.665, interpMode=RCIM_Cubic, leaveTangent=-2.665, tangentMode=RCTM_User, time=0, value=1}, <at 1 0.4 RCIM_Cubic>]}}}
       Scale Curve=150
      ))
     Alpha Scale:NiagaraFloat=1
    )
    Particles.EMBERS_UP_SCALE:Vector2f=&Multiply_Vector2D_ByFloat(
     Float=&FloatFromCurve(
      FloatCurve=DataInterface() {propertyValues {Curve {
         defaultValue=10
         keys [
          {arriveTangent=-0.1384, interpMode=RCIM_Cubic, leaveTangent=-0.1384, tangentMode=RCTM_User, time=0, value=0}
          {arriveTangent=-0.1738, interpMode=RCIM_Cubic, leaveTangent=-0.1738, tangentMode=RCTM_User, time=0.2, value=1}
          {arriveTangent=-0.0758, interpMode=RCIM_Cubic, leaveTangent=-0.0758, tangentMode=RCTM_User, time=0.9, value=0.5}
          {arriveTangent=-0.0405, interpMode=RCIM_Cubic, leaveTangent=-0.0405, tangentMode=RCTM_User, time=1, value=0.2}
         ]
        }}}
      Scale Curve=18
     )
     Vector2D=<0.15 0.9>
    )
    Particles.FLAMES_SCALE:Vector2f=&Multiply_Vector2D_ByFloat(
     Float=&FloatFromCurve(
      FloatCurve=DataInterface() {propertyValues {Curve {
         defaultValue=10
         keys [{arriveTangent=0.1028, interpMode=RCIM_Cubic, leaveTangent=0.1028, tangentMode=RCTM_User, time=0.1, value=0.5}, <at 1 0 RCIM_Cubic>]
        }}}
      Scale Curve=&RandomRangeFloat(Minimum=50, Maximum=60)
     )
     Vector2D=<1 1>
    )
    Particles.Flames Color:LinearColor=&VC_DynamicScaleColor(
     Color to Scale:LinearColor=$Particles.Initial.R2 Color
     RGB Scale:Vector3f=&VectorFromFloat(Value=&FloatFromCurve(FloatCurve=DataInterface() {propertyValues {Curve {keys [<at 0 1 RCIM_Cubic>,<at 1 0 RCIM_Cubic>]}}}, Scale Curve=100))
     Alpha Scale:NiagaraFloat=2
    )
   ) {isSetParameters=true}
  ]
  rendererClasses [&NiagaraSpriteRendererProperties]
  renderers [
   {rendererClass=&NiagaraSpriteRendererProperties}
   {rendererClass=&NiagaraSpriteRendererProperties, rendererIndex=1}
   {rendererClass=&NiagaraSpriteRendererProperties, rendererIndex=2}
  ]
 }
 Wheel {
  emitterUpdateScript [
   &EmitterState(
    Inactive Response=::NewEnumerator1
    Life Cycle Mode=::NewEnumerator1
    Loop Behavior=::NewEnumerator1
    Loop Duration Mode=::NewEnumerator1
    UseLoopDelay=false
   )
   &SpawnBurst_Instantaneous(Spawn Count=2, Use Spawn Probability=false, UseLoopCountLimit=false)
   &SpawnRate(SpawnGroup=1, SpawnRate=2, Use Spawn Probability=false) {enabled=false}
  ]
  particleSpawnScript [
   &VC_PaintedVariantsOverride() {enabled=false}
   &VC_PaintedVariantsOverride(
    Black=<rgba 0.015 0.015 0.015 1>
    BurntSienna=<rgba 0.18 0.033 0 1>
    Cobalt=<rgba 0 0.1667 2 1>
    Crimson=<rgba 2 0.02 0.02 1>
    DefaultColor=<rgba 4 0.3627 0.032 1>
    ForestGreen=<rgba 0 0.4 0.02 1>
    Grey=<rgba 0.05 0.05 0.05 1>
    Lime=<rgba 0.3333 1 0 1>
    Pink=<rgba 3 0.3594 2.0398 1>
    Purple=<rgba 0.1995 0.0347 0.6939 1>
    Saffron=<rgba 1 0.7713 0.02 1>
    SkyBlue=<rgba 0.02 0.755 1 1>
   ) {module=VC_PaintedVariantsOverride001}
   /SetVariables(Particles.RenderVisTag:NiagaraInt32=&SpawnGroup()) {isSetParameters=true}
   &InitializeParticle(
    Color Mode=::NewEnumerator0
    Lifetime=0
    Mesh Scale Mode=::NewEnumerator3
    Sprite Rotation Mode=::NewEnumerator2
    Sprite Size=<0.5 5>
    Sprite Size Mode=::NewEnumerator4
    UsePositionOffset=false
    Write Material Random=true
   )
   &PartitionParticles()
   &SkeletalMeshLocation(
    Bone / Socket Index=&Add_Integer(A=$Particles.PartitionParticles.Partition, B=4)
    Bone / Socket Sampling Mode=::NewEnumerator5
    Interpolate Mesh Sampling=false
    MaskSpawnGroup=false
    Mesh Sampling Type=::NewEnumerator6
    Skeletal Mesh=$System.SkeletalMesh
   )
   &ShapeLocation(
    Handle Radius:NiagaraFloat=&VC_FloatBySpawnGroup(
     SpawnGroup or ExecIndex:NiagaraInt32=0
     CustomInt:NiagaraInt32=0
     MaximumSpawnGroups:NiagaraInt32=2
     Float SG 0:NiagaraFloat=0
     Float SG 1:NiagaraFloat=10
     Float SG 2:NiagaraFloat=0
     Float SG 3:NiagaraFloat=0
     Float SG 4:NiagaraFloat=0
     Float SG 5:NiagaraFloat=0
    )
    Large Radius=&VC_FloatBySpawnGroup(
     SpawnGroup or ExecIndex:NiagaraInt32=0
     CustomInt:NiagaraInt32=0
     MaximumSpawnGroups:NiagaraInt32=2
     Float SG 0:NiagaraFloat=0
     Float SG 1:NiagaraFloat=45
     Float SG 2:NiagaraFloat=0
     Float SG 3:NiagaraFloat=0
     Float SG 4:NiagaraFloat=0
     Float SG 5:NiagaraFloat=0
    )
    MaskSpawnGroup=false
    Override Seed=false
    Rotation Axis=<0 1 0>
    Rotation Mode=::NewEnumerator1
    Shape Primitive=::NewEnumerator3
    UseOwnerScale=true
   )
   &AddVelocity(
    Constrain To Radius=1500
    Module.Use Falloff Distance=true
    Origin Offset=<1 0 0>
    Origin Offset Coordinate Space=::Simulation
    Override Seed=false
    Radius Falloff Exponent=0
    Velocity Mode=::NewEnumerator1
    Velocity Speed=&VC_FloatBySpawnGroup(
     SpawnGroup or ExecIndex:NiagaraInt32=0
     CustomInt:NiagaraInt32=0
     MaximumSpawnGroups:NiagaraInt32=2
     Float SG 0:NiagaraFloat=0
     Float SG 1:NiagaraFloat=150000
     Float SG 2:NiagaraFloat=0
     Float SG 3:NiagaraFloat=0
     Float SG 4:NiagaraFloat=0
     Float SG 5:NiagaraFloat=0
    )
   )
  ]
  particleUpdateScript [
   &FN_DisableTireFX(
    Mask Index Attribute:NiagaraInt32=$Particles.PartitionParticles.Partition
    Mask Index FL=2
    Mask Index FR=3
    Mask Index RL=0
    Mask Index RR=1
   )
   &VC_WheelWrapScale(Wheel Index Attribute:NiagaraInt32=&Add_Integer(A=$Particles.PartitionParticles.Partition, B=2))
   &ParticleState(Kill Particles When Lifetime Has Elapsed=false)
   &ScaleColor(Scale RGB=&Multiply_VectorByFloat(Vector=<1 1 1>, Float=400), ScaleA=true, ScaleRGB=true)
   &UpdateMeshOrientation(Rotation Coordinate Space=::NewEnumerator2, Rotation Vector=<0 1 0>) {enabled=false}
   &SkeletalMeshLocation(
    Bone / Socket Index=&Add_Integer(A=$Particles.PartitionParticles.Partition, B=4)
    Bone / Socket Sampling Mode=::NewEnumerator5
    MaskSpawnGroup=false
    Mesh Sampling Type=::NewEnumerator6
    Skeletal Mesh=$System.SkeletalMesh
   ) {module=SkeletalMeshLocation001}
   &SolveForcesAndVelocity(Clamp Velocity=false, Limit Acceleration=false)
  ]
  rendererClasses [&NiagaraMeshRendererProperties]
  renderers [{rendererClass=&NiagaraMeshRendererProperties, rendererIndex=0}]
 }
}
summary {
 systemName=NS_DriftTrail_Chains
 emitters [
  {emitterName=BounceSparks_Generate, rendererClasses [&NiagaraSpriteRendererProperties]}
  {emitterName=Chains, rendererClasses [&NiagaraRibbonRendererProperties]}
  {emitterName=BounceSparks_DeathEvent001, rendererClasses [&NiagaraSpriteRendererProperties]}
  {emitterName=Wheel, rendererClasses [&NiagaraMeshRendererProperties]}
  {emitterName=NE_VC_NEW_SPARKS, rendererClasses [&NiagaraSpriteRendererProperties]}
 ]
 userVariables [
  User.bIsFrontendPreview {type=NiagaraBool}
  User.bIsPainted {type=NiagaraBool}
  User.ColorVariantID {type=NiagaraInt32}
  User.offset {type=Vector3f, description="offset from wheels(set through cosmetic template)", defaultValue {struct=Vector3f, value=<0 0 0>}}
  User.PaintedColor {type=LinearColor, description="", defaultValue {struct=LinearColor, value=<rgba 1 1 1 1>}}
  User.PaintedScalar {type=NiagaraFloat, defaultValue=1, description=""}
  User.VehicleFlags {type=NiagaraInt32, defaultValue=806289408}
  User.Wheel_Radius_Array {
   type=NiagaraDataInterfaceArrayFloat
   defaultValue {
    struct=NiagaraExt_VariableValue_DataInterface
    value {
     dataInterface=&self:NiagaraDataInterfaceArrayFloat_1
     dataInterfaceClass=&NiagaraDataInterfaceArrayFloat
    }
   }
  }
  User.Wheel_Width_Array {
   type=NiagaraDataInterfaceArrayFloat
   defaultValue {
    struct=NiagaraExt_VariableValue_DataInterface
    value {
     dataInterface=&self:NiagaraDataInterfaceArrayFloat_0
     dataInterfaceClass=&NiagaraDataInterfaceArrayFloat
    }
   }
  }
  User.WheelTouchingArray {
   type=NiagaraDataInterfaceArrayFloat
   defaultValue {
    struct=NiagaraExt_VariableValue_DataInterface
    value {
     dataInterface=&self:NiagaraDataInterfaceArrayFloat_2
     dataInterfaceClass=&NiagaraDataInterfaceArrayFloat
    }
   }
  }
 ]
}
