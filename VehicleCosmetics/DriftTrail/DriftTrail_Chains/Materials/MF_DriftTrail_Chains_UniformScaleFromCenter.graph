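# Material function in the Chains drift trail cosmetic that scales UV coordinates around the 0.5 center point.
# It divides the input coordinates, defaulting to the texture coordinate, by the scale, then re-centers them, so larger values enlarge whatever is sampled with them.
# A static switch selects either separate U and V scales or a single uniform scale.
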MaterialFunction /VehicleCosmetics/DriftTrail/DriftTrail_Chains/Materials/MF_DriftTrail_Chains_UniformScaleFromCenter
graph {nodes [
  n1 = FunctionInput() {bUsePreviewValueAsDefault=true, inputName=U_Scale, inputType=FunctionInput_Scalar, sortPriority=2, previewValue=<1 0 0 1>}
  n2 = FunctionInput() {bUsePreviewValueAsDefault=true, inputName=V_Scale, inputType=FunctionInput_Scalar, sortPriority=3, previewValue=<1 0 0 1>}
  n3 = FunctionInput() {bUsePreviewValueAsDefault=true, inputName=UV_Scale, inputType=FunctionInput_Scalar, sortPriority=4, previewValue=<1 0 0 1>}
  n4 = StaticSwitch(
   True=AppendVector(A=@n1, B=@n2)
   False=@n3
   Value=FunctionInput(Preview=StaticBool()) {bUsePreviewValueAsDefault=true, inputName=Scale_UVs_Separately, inputType=FunctionInput_StaticBool, sortPriority=1}
  )
  n5 = Divide(A=FunctionInput(Preview=TextureCoordinate()) {bUsePreviewValueAsDefault=true, inputName=Coordinates, inputType=FunctionInput_Vector2}, B=@n4)
  n6 = Constant() {r=0.5}
  FunctionOutput(None=Subtract(A=Add(A=@n5, B=@n6), B=Divide(A=@n6, B=@n4)))
 ]}
