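# Material function that returns an animated pseudo-random 2D gradient vector per grid position, a helper for pseudo-3D noise in the Chains drift trail cosmetic.
# It hashes the IntPos(UV) input with a classic dot, sine and frac hash.
# The hash sets both a base angle and a spin rate scaled by input t, so each gradient rotates at its own speed as t changes.
# Cosine and sine of that angle form the unit-vector output.
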MaterialFunction /VehicleCosmetics/DriftTrail/DriftTrail_Chains/Materials/MF_DriftTrail_Chains_Noise_Pseudo3D_GetGradient
graph {nodes [
  n1 = DotProduct(
   A=FunctionInput(Preview=TextureCoordinate()) {bUsePreviewValueAsDefault=true, inputName="IntPos(UV)", inputType=FunctionInput_Vector2}
   B=Constant2Vector() {g=78.233, r=12.9898}
  )
  n2 = Frac(None=Multiply(A=Sine(None=@n1), B=Constant() {r=43758.543}))
  n3 = Constant() {r=4}
  n4 = FunctionInput() {bUsePreviewValueAsDefault=true, inputName=t, inputType=FunctionInput_Scalar, sortPriority=1, previewValue=<1 0 0 1>}
  n5 = Add(A=Multiply(A=@n2, B=Constant() {r=6.2832}), B=Multiply(A=Multiply(A=@n2, B=@n3), B=@n4))
  FunctionOutput(None=AppendVector(A=Cosine(None=@n5), B=Sine(None=@n5)))
 ]}
