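# Material function for the Chains drift trail vehicle cosmetic that generates animated pseudo-3D gradient noise.
# Scaled, aspect-adjusted UVs split into cell and fraction, while Time feeds a GetGradient sub-function for each of four corners.
# Dot products with those gradients blend bilinearly, apparently using smoothstep weights, creating Perlin-style noise without a full 3D lattice.
# Outputs provide an absolute-value plasma variant plus remapped noise as float and float3.
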MaterialFunction /VehicleCosmetics/DriftTrail/DriftTrail_Chains/Materials/MF_DriftTrail_Chains_Noise_Pseudo3D
graph {nodes [
  n1 = Constant2Vector()
  n2 = Add(A=NamedRerouteUsage(), B=@n1)
  n3 = FunctionInput() {bUsePreviewValueAsDefault=true, inputName=Scale, inputType=FunctionInput_Scalar, sortPriority=1, previewValue=<1 0 0 1>}
  n4 = Multiply(A=FunctionInput(Preview=TextureCoordinate()) {bUsePreviewValueAsDefault=true, inputName=UV, inputType=FunctionInput_Vector2}, B=@n3)
  n5 = FunctionInput(Preview=Constant2Vector() {g=1, r=1}) {bUsePreviewValueAsDefault=true, inputName=AspectRatio, inputType=FunctionInput_Vector2, sortPriority=2}
  n6 = AppendVector(
   A=Multiply(A=Multiply(A=@n4, B=@n5), B=Constant() {r=10})
   B=FunctionInput(Preview=Time()) {bUsePreviewValueAsDefault=true, inputName=Time, inputType=FunctionInput_Scalar, sortPriority=3}
  )
  n7 = ComponentMask(None=@n6) {b=true}
  n8 = NamedRerouteDeclaration(None=@n7) {name=MaskB, nodeColor=<rgba 0.6406 1 0 1>}
  n9 = DotProduct(
   A=MaterialFunctionCall!("IntPos(UV)"=AppendVector(A=@n2, B=@n7), t=@n8) {materialFunction=&/MF_DriftTrail_Chains_Noise_Pseudo3D_GetGradient}.Result
   B=Subtract(A=NamedRerouteUsage(), B=@n1)
  )
  n10 = Constant2Vector() {r=1}
  n11 = Add(A=NamedRerouteUsage(), B=@n10)
  n12 = DotProduct(
   A=MaterialFunctionCall("IntPos(UV)"=AppendVector(A=@n11, B=@n8), t=@n8) {materialFunction=&MF_DriftTrail_Chains_Noise_Pseudo3D_GetGradient}.Result
   B=Subtract(A=NamedRerouteUsage(), B=@n10)
  )
  n13 = ComponentMask(None=NamedRerouteUsage()) {r=true}
  n14 = NamedRerouteUsage()
  n15 = Constant2Vector() {g=1}
  n16 = DotProduct(
   A=MaterialFunctionCall("IntPos(UV)"=Add(A=@n14, B=@n15), t=NamedRerouteUsage()) {materialFunction=&MF_DriftTrail_Chains_Noise_Pseudo3D_GetGradient}.Result
   B=Subtract(A=NamedRerouteUsage(), B=@n15)
  )
  n17 = NamedRerouteUsage()
  n18 = Constant2Vector() {g=1, r=1}
  n19 = DotProduct(
   A=MaterialFunctionCall("IntPos(UV)"=Add(A=@n17, B=@n18), t=NamedRerouteUsage()) {materialFunction=&MF_DriftTrail_Chains_Noise_Pseudo3D_GetGradient}.Result
   B=Subtract(A=NamedRerouteUsage(), B=@n18)
  )
  n20 = NamedRerouteUsage()
  n21 = Divide(
   A=LinearInterpolate(A=LinearInterpolate(A=@n9, B=@n12, Alpha=@n13), B=LinearInterpolate(A=@n16, B=@n19, Alpha=@n13), Alpha=ComponentMask(None=@n20) {g=true})
   B=Constant() {r=0.7}
  )
  FunctionOutput(None=Abs(None=@n21)) {outputName=Float Plasma1}
  n22 = Constant() {r=0.5}
  n23 = Add(A=Multiply(A=@n21, B=@n22), B=@n22)
  FunctionOutput(None=AppendVector(A=AppendVector(A=@n23, B=@n23), B=@n23)) {outputName=Float3 Noise}
  FunctionOutput(None=@n23) {outputName=Float Noise}
  n24 = ComponentMask(None=@n6) {g=true, r=true}
  n25 = Subtract(A=@n24, B=NamedRerouteDeclaration(None=Floor(None=@n24)) {name=Floor, nodeColor=<rgba 0.6406 1 0 1>})
  n26 = Subtract(A=Constant() {r=3}, B=Multiply(A=NamedRerouteDeclaration(None=@n25) {name=Subtract, nodeColor=<rgba 0.6406 1 0 1>}, B=Constant() {r=2}))
  n27 = NamedRerouteUsage()
  NamedRerouteDeclaration(None=Multiply(A=Multiply(A=@n26, B=@n27), B=@n27)) {name=Multiply, nodeColor=<rgba 0.6406 1 0 1>}
 ]}
