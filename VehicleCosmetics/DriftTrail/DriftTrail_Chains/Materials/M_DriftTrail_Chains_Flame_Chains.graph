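# Translucent, two-sided unlit VFX material for the Chains drift trail cosmetic, blending flame and chain looks via ISFLAMES? (default 1 selects chains).
# Flames multiply panning smoke by cloud noise.
# Chains use a scrolling, sine-distorted alpha with depth fade, shaded by a normal-driven rusty-metal matcap and optional derivative-based specular, tinted by particle color.
# A WPO network builds noise-offset, switchable-spin waves with end masks and optional camera-distance fade.
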Material /VehicleCosmetics/DriftTrail/DriftTrail_Chains/Materials/M_DriftTrail_Chains_Flame_Chains
graph {
 properties {
  bAutomaticallySetUsageInEditor=false
  bUsedWithMeshDeformer=true
  bUsedWithMeshParticles=true
  bUsedWithNiagaraMeshParticles=true
  bUsedWithNiagaraRibbons=true
  bUsedWithNiagaraSprites=true
  bUsedWithParticleSprites=true
  bUsedWithSkeletalMesh=true
  bUsedWithSplineMeshes=true
  blendMode=BLEND_Translucent
  shadingModel=MSM_Unlit
  twoSided=true
  usedShadingModels=Unlit
 }
 nodes [
  n1 = TextureCoordinate() {uTiling=3, unMirrorU=true, vTiling=0.75}
  LinearInterpolate(A=@n1, B=TextureCoordinate() {uTiling=2.8, unMirrorU=true, vTiling=0.75}, Alpha=Sine(None=Time()) {period=0.6}) {constA=2, constB=3}
  n2 = TextureCoordinate() {uTiling=1.5}
  LinearInterpolate(A=@n2, B=TextureCoordinate() {uTiling=1.3}, Alpha=Sine(None=Time()) {period=0.4}) {constA=2, constB=3}
  LinearInterpolate(Alpha=Sine(None=Time()) {period=0.25}) {constA=0.2, constB=0.5}
  n3 = ParticleColor()
  n4 = Panner(
   Coordinate=@n2
   Speed=MaterialFunctionCall(X=LinearInterpolate() {constA=0.5, constB=-4}, Y=Constant() {r=0.5}) {materialFunction=&MakeFloat2}.Result
  )
  n5 = Multiply(
   A=TextureSample(UVs=Panner(Coordinate=@n1) {speedX=-2}) {texture=/DealershipSharedLibrary/FX/Textures/TexturePackage/T_VCVFX_SmokeTrail01_RGB}.R
   B=Power(Base=TextureSample(UVs=@n4) {texture=/DealershipSharedLibrary/FX/Textures/Noise/T_VCVFX_CloudDistort_A}.RGB) {constExponent=0.5}
  )
  n6 = TextureCoordinate()
  n7 = Power(Base=Sine(None=ComponentMask(None=@n6) {g=true}) {period=2}) {constExponent=3}
  n8 = OneMinus(None=ComponentMask(None=@n6) {r=true})
  n9 = Multiply(A=@n8, B=@n8)
  n10 = ParticleColor()
  n11 = ComponentMask(None=Multiply(A=TextureCoordinate(), B=ScalarParameter() {defaultValue=1, group=UV Wave, parameterName=Sin_B_UVTile})) {r=true}
  n12 = Multiply(A=ScalarParameter() {defaultValue=1.25, group=UV Wave, parameterName=Sin_B_UVTimeMult}, B=Time())
  n13 = ScalarParameter() {group=UV Wave, parameterName=Sin_B_UVTimeOffset}
  n14 = Fmod(A=Add(A=Add(A=@n12, B=@n13), B=NamedRerouteUsage()), B=Constant() {r=1})
  n15 = Multiply(A=Sine(None=Add(A=@n11, B=@n14)), B=ScalarParameter() {defaultValue=0.1, group=UV Wave, parameterName=Dist_Wave_Z_Mult})
  n16 = ComponentMask(None=Multiply(A=TextureCoordinate(), B=ScalarParameter() {defaultValue=1, group=UV Wave, parameterName=SinUVTile})) {r=true}
  n17 = Multiply(A=ScalarParameter() {defaultValue=1, group=UV Wave, parameterName=SinUVTimeMult}, B=Time())
  n18 = ScalarParameter() {group=UV Wave, parameterName=SinUVTimeOffset}
  n19 = Fmod(A=Add(A=Add(A=@n17, B=@n18), B=NamedRerouteUsage()), B=Constant() {r=1})
  n20 = Multiply(A=ScalarParameter() {defaultValue=0.1, group=UV Wave, parameterName=Dist_Wave_Y_Mult}, B=Sine(None=Add(A=@n16, B=@n19)))
  n21 = Multiply(
   A=Add(
    A=StaticSwitchParameter(True=@n15, False=@n20) {parameterName=Spin_CCW}
    B=StaticSwitchParameter(True=@n20, False=@n15) {parameterName=Spin_CCW}
   )
   B=ScalarParameter() {defaultValue=3, group=UV Wave, parameterName=Dist_WaveMult}
  )
  n22 = ScalarParameter() {defaultValue=0.25, group=UV Wave, parameterName=Dist_Mask_Min}
  n23 = TextureCoordinate() {coordinateIndex=1}
  n24 = Multiply(
   A=Power(Base=ComponentMask(None=@n23) {r=true}, Exp=ScalarParameter() {defaultValue=0.5, group=UV Wave, parameterName=Dis_Mask_Power})
   B=ScalarParameter() {defaultValue=4, group=UV Wave, parameterName=Dist_Mask_Mult}
  )
  n25 = StaticSwitchParameter(True=Saturate(None=Add(A=@n22, B=@n24)), False=Constant() {r=1}) {parameterName=Dist_WaveMask}
  n26 = TextureCoordinate()
  n27 = Multiply(
   A=Power(Base=ComponentMask(None=@n26) {r=true}, Exp=ScalarParameter() {defaultValue=1, group=UV Wave, parameterName=Dist_MaskPower})
   B=ScalarParameter() {defaultValue=1, group=UV Wave, parameterName=Dist_MaskMult}
  )
  n28 = AppendVector(A=Constant(), B=Multiply(A=Multiply(A=@n21, B=@n25), B=Saturate(None=@n27)))
  n29 = NamedRerouteUsage()
  n30 = MaterialFunctionCall(
   UVs=TextureCoordinate()
   Texture Scale=MaterialFunctionCall(X=ScalarParameter() {defaultValue=1, parameterName=Dist_Pre_U_Scale}, Y=ScalarParameter() {defaultValue=0.8, parameterName=Dist_Pre_V_Scale}) {materialFunction=&MakeFloat2}.Result
  ) {materialFunction=&ScaleUVsByCenter}
  n31 = AppendVector(A=ScalarParameter() {defaultValue=1, parameterName=Dist_U_Scale}, B=ScalarParameter() {defaultValue=1, parameterName=Dist_V_Scale})
  n32 = Time()
  n33 = ScalarParameter() {defaultValue=1, parameterName=Dist_ScrollMult}
  n34 = Add(
   A=StaticSwitchParameter(True=Multiply(A=@n28, B=@n29), False=Constant()) {defaultValue=true, parameterName=Use_UV_Distortion}
   B=Panner(Coordinate=Multiply(A=@n30.UVs, B=@n31), Time=Multiply(A=@n32, B=@n33)) {speedX=1}
  ) {constB=0}
  n35 = ComponentMask(None=TextureCoordinate()) {g=true}
  n36 = OneMinus(None=@n35)
  n37 = Multiply(
   A=Multiply(A=@n10.A, B=TextureSample(UVs=@n34) {texture=/VehicleCosmetics/DriftTrail/DriftTrail_Chains/Textures/T_VCVFX_Chain_02A_Alpha}.R)
   B=Step(Y=ScalarParameter() {group=Opacity, parameterName=OpacityUVMask}, X=Multiply(A=@n35, B=@n36)) {constY=0.22}
  )
  n38 = ScalarParameter() {defaultValue=1, parameterName=ISFLAMES?}
  LinearInterpolate(
   A=Multiply(A=Multiply(A=@n3.A, B=@n5), B=Multiply(A=@n7, B=@n9))
   B=MaterialFunctionCall(Opacity In=Saturate(None=@n37), DepthFade Distance=ScalarParameter() {group=DepthBlend, parameterName=DepthBlendDistance}) {materialFunction=&MF_VCVFX_DepthFade}.Result
   Alpha=@n38
  )
  n39 = TextureSampleParameter2D(UVs=@n34) {
   group=Specular
   parameterName=NormalMap
   samplerType=SAMPLERTYPE_Normal
   texture=/VehicleCosmetics/DriftTrail/DriftTrail_Chains/Textures/T_VCVFX_Chain_02A_Normals
  }
  n40 = Normalize(VectorInput=Transform(None=@n39.R) {transformType=TRANSFORM_View})
  n41 = WorldPosition()
  n42 = DDX(Value=@n41.XYZ)
  n43 = DDY(Value=@n41.XYZ)
  n44 = Normalize(VectorInput=Abs(None=CrossProduct(A=@n42, B=@n43)))
  n45 = Multiply(
   A=Power(Base=DotProduct(A=@n40, B=@n44), Exp=ScalarParameter() {defaultValue=1, group=Specular, parameterName=SpecPower})
   B=ScalarParameter() {defaultValue=1, group=Specular, parameterName=SpecMult}
  )
  n46 = Constant() {r=1}
  n47 = TextureSampleParameter2D(UVs=@n34) {
   group=Specular
   parameterName=NormalMap
   samplerType=SAMPLERTYPE_Normal
   texture=&T_VCVFX_Chain_02A_Normals
  }
  n48 = MaterialFunctionCall(Color=Transform(None=@n47.RGB) {transformSourceType=TRANSFORMSOURCE_Camera, transformType=TRANSFORM_Local}) {materialFunction=&SplitComponents}
  n49 = Constant() {r=1}
  n50 = Multiply(A=Add(A=Constant() {r=1}, B=@n48.R), B=@n49)
  n51 = Constant() {r=10}
  n52 = Multiply(A=Multiply(A=@n49, B=Add(A=@n48.G, B=@n51)), B=Constant() {r=1})
  n53 = Add(
   A=Clamp(None=StaticSwitchParameter(True=@n45, False=@n46) {parameterName=AddSpec}) {maxDefault=0}
   B=TextureSampleParameter2D(UVs=AppendVector(A=@n50, B=@n52)) {parameterName=MatCap_Texture, texture=/VehicleCosmetics/DriftTrail/DriftTrail_Chains/Textures/T_VCVFX_Rusty_Metal_01_TX}.R
  )
  n54 = Multiply(A=@n10.RGB, B=Multiply(A=@n53, B=@n53) {constB=0})
  MaterialFunctionCall(Emissive=LinearInterpolate(A=Multiply(A=@n3.RGB, B=@n5), B=Saturate(None=@n54), Alpha=@n38)) {materialFunction=&MF_FX_EmissiveAdjustments}
  n55 = DynamicParameter() {paramNames [Param1, Param2, TimeOffset, WaveMultOverTime]}
  NamedRerouteDeclaration(None=@n55.TimeOffset) {name=Dynamic - TimeOffset, nodeColor=<rgba 0 1 0.3906 1>}
  NamedRerouteDeclaration(None=@n55.WaveMultOverTime) {name=Dynamic - WaveMultOverTime, nodeColor=<rgba 0 1 0.3906 1>}
  NamedRerouteDeclaration(None=ScalarParameter() {group=WPO, parameterName=WPO_TimeOffset}) {name=WPO_TimeOffset, nodeColor=<rgba 0.2789 0 0.6795 1>}
  n56 = ComponentMask(None=TextureCoordinate()) {r=true}
  n57 = Multiply(
   A=Power(Base=@n56, Exp=ScalarParameter() {defaultValue=0.5, group=WPO, parameterName=WPO_Start_WaveMaskPOW})
   B=ScalarParameter() {defaultValue=1, group=WPO, parameterName=WPO_Start_WaveMaskMult}
  )
  n58 = Power(Base=OneMinus(None=@n56), Exp=ScalarParameter() {defaultValue=0.5, group=WPO, parameterName=WPO_End_WaveMaskPOW})
  n59 = ScalarParameter() {defaultValue=1, group=WPO, parameterName=WPO_End_WaveMaskMult}
  n60 = Saturate(None=Add(A=Multiply(A=@n58, B=@n59), B=ScalarParameter() {group=WPO, parameterName=WPO_End_WaveMaskAdd}))
  NamedRerouteDeclaration(None=Saturate(None=Multiply(A=@n57, B=@n60))) {name=WPO_Mask, nodeColor=<rgba 0.2789 0 0.6795 1>}
  n61 = TextureCoordinate()
  n62 = ComponentMask(None=@n61) {r=true}
  n63 = Time()
  n64 = ScalarParameter() {defaultValue=1, parameterName=NoiseSpeed}
  n65 = MaterialFunctionCall!(
   UV=AppendVector(A=@n62, B=Constant())
   Scale=ScalarParameter() {defaultValue=1, parameterName=NoiseScale}
   AspectRatio
   Time=Add(A=ScalarParameter() {parameterName=NoiseTimeOffset}, B=Multiply(A=@n63, B=@n64))
  ) {materialFunction=&/MF_DriftTrail_Chains_Noise_Pseudo3D}
  n66 = Constant()
  n67 = ScalarParameter() {defaultValue=1, parameterName=NoiseMax}
  n68 = ScalarParameter() {defaultValue=2, group=WPO, parameterName=WPO_WaveTile}
  n69 = ScalarParameter() {defaultValue=1, group=WPO, parameterName=WPO_WaveSpeed}
  n70 = Time()
  n71 = Add(
   A=Add(A=Clamp(None=@n65.Float Noise, Min=@n66, Max=@n67), B=Multiply(A=@n62, B=@n68))
   B=Add(A=Multiply(A=@n69, B=@n70), B=NamedRerouteUsage())
  )
  n72 = Sine(None=@n71)
  n73 = Cosine(None=@n71)
  n74 = ComponentMask(None=@n61) {r=true}
  n75 = ScalarParameter() {defaultValue=1, group=WPO, parameterName=WPO_Wave_B_Tile_1}
  n76 = Time()
  n77 = Multiply(A=ScalarParameter() {defaultValue=0.5, group=WPO, parameterName=WPO_Wave_B_Speed_1}, B=@n76)
  n78 = NamedRerouteUsage()
  n79 = Sine(None=Add(A=Multiply(A=@n74, B=@n75), B=Add(A=@n77, B=@n78)))
  n80 = ScalarParameter() {defaultValue=0.5, group=WPO, parameterName=WPO_Wave_B_Tile_2}
  n81 = Multiply(A=ScalarParameter() {defaultValue=0.5, group=WPO, parameterName=WPO_Wave_B_Speed_2}, B=@n76)
  n82 = Cosine(None=Add(A=Multiply(A=@n74, B=@n80), B=Add(A=@n81, B=@n78)))
  n83 = Multiply(
   A=Multiply(
    A=StaticSwitchParameter(True=@n72, False=@n73) {group=WPO, parameterName=WPO_Spin_CCW}
    B=ScalarParameter() {defaultValue=1, group=WPO, parameterName=WPO_WaveScaleY}
   )
   B=Multiply(
    A=StaticSwitchParameter(True=@n79, False=@n82) {group=WPO, parameterName=WPO_B_Spin_CCW}
    B=ScalarParameter() {defaultValue=1, group=WPO, parameterName=WPO_Wave_B_ScaleY}
   )
  )
  n84 = Time()
  n85 = ScalarParameter() {defaultValue=1, group=WPO, parameterName=WPO_ScaleOverTime_Speed}
  n86 = Sine(None=Add(A=NamedRerouteUsage(), B=Multiply(A=@n84, B=@n85)))
  n87 = Multiply(
   A=ScalarParameter() {defaultValue=1, group=WPO, parameterName=WPO_Wave_Overall_Scale_Adjust}
   B=Clamp(
    None=ConstantBiasScale(None=@n86)
    Min=ScalarParameter() {defaultValue=0.5, group=WPO, parameterName=WPO_ScaleOverTime_Min}
    Max=ScalarParameter() {defaultValue=1, group=WPO, parameterName=WPO_ScaleOverTime_Max}
   )
  )
  n88 = NamedRerouteUsage()
  n89 = Add(A=Multiply(A=Multiply(A=@n83, B=@n87), B=@n88), B=ScalarParameter() {group=WPO, parameterName=WPO_Wave_Shift_Y})
  n90 = NamedRerouteUsage()
  n91 = Multiply(
   A=Multiply(
    A=StaticSwitchParameter(True=@n73, False=@n72) {group=WPO, parameterName=WPO_Spin_CCW}
    B=ScalarParameter() {defaultValue=1, group=WPO, parameterName=WPO_WaveScaleX}
   )
   B=Multiply(
    A=StaticSwitchParameter(True=@n82, False=@n79) {group=WPO, parameterName=WPO_B_Spin_CCW}
    B=ScalarParameter() {defaultValue=1, group=WPO, parameterName=WPO_Wave_B_ScaleX}
   )
  )
  n92 = Multiply(A=Multiply(A=Multiply(A=@n91, B=@n87), B=@n88), B=@n90)
  n93 = OneMinus(None=@n92)
  n94 = Transform(None=MaterialFunctionCall(R=Constant(), G=Multiply(A=@n89, B=@n90), B=StaticSwitchParameter(True=@n93, False=@n92) {group=WPO, parameterName=MirrorTest}) {materialFunction=&AppendMany}.RGB) {transformSourceType=TRANSFORMSOURCE_Local}
  NamedRerouteDeclaration(None=Multiply(A=ConstantBiasScale(None=@n94), B=VectorParameter() {group=WPO, parameterName=WPO_WorldSpaceMult, defaultValue=<rgba 1 1 1 1>}.RGB)) {name=WPO, nodeColor=<rgba 0.2789 0 0.6795 1>}
  n95 = MaterialFunctionCall(
   Input=Distance(A=WorldPosition().XYZ, B=CameraPositionWS())
   Input Low=ScalarParameter() {group=WPO, parameterName=DistFadeWavesMin}
   Input High=ScalarParameter() {defaultValue=440, group=WPO, parameterName=DistFadeWavesMax}
   Target Low=Constant()
   Target High=Constant() {r=1}
  ) {materialFunction=&RemapValueRange}
  n96 = OneMinus(None=@n95.Result)
  NamedRerouteDeclaration(None=StaticSwitchParameter(True=StaticSwitchParameter(True=@n95.Result, False=@n96) {group=WPO, parameterName=InvertDistanceFadeWaves}, False=Constant() {r=1}) {group=WPO, parameterName=DistanceFadeWaves}) {name=WPO_DistanceFade, nodeColor=<rgba 0.2789 0 0.6795 1>}
  Transform(None=@n47.RGB) {transformSourceType=TRANSFORMSOURCE_World, transformType=TRANSFORM_View}
 ]
}
