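# Unlit translucent Niagara sprite and ribbon material for the tire fire effect in the Chains drift trail cosmetic.
# Panning curl noise offsets UVs that sample a tiled wispy smoke texture and also warp an off-center radial mask, shaped by Size and MaskPow.
# Smoke, mask and particle alpha multiply into an opacity term.
# Particle color drives emissive through MF_FX_EmissiveAdjustments, as the smoke-driven color lerp blends two identical whites.
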Material /VehicleCosmetics/DriftTrail/DriftTrail_Chains/Materials/M_DriftTrail_Chains_Tire_Fire
graph {
 properties {
  bAutomaticallySetUsageInEditor=false
  bUsedWithNiagaraRibbons=true
  bUsedWithNiagaraSprites=true
  blendMode=BLEND_Translucent
  shadingModel=MSM_Unlit
  usedShadingModels=Unlit
 }
 nodes [
  n1 = Panner(Coordinate=TextureCoordinate() {uTiling=0.1, unMirrorU=true, unMirrorV=true, vTiling=0.5}) {speedX=-1, speedY=1}
  n2 = Add(
   A=TextureCoordinate()
   B=ComponentMask(None=TextureSample(UVs=@n1) {texture=/DealershipSharedLibrary/FX/Textures/Noise/T_VCVFX_Flow_Curl_01}.RGB) {g=true, r=true}
  )
  n3 = Multiply(A=Constant() {r=3}, B=@n2)
  n4 = TextureSample(UVs=@n3) {texture=/VehicleCosmetics/DriftTrail/DriftTrail_Chains/Textures/T_VCVFX_Smoke_Wispy_Circle_A}
  n5 = LinearInterpolate(A=Constant3Vector() {constant=<rgba 1 1 1 0>}, B=Constant3Vector() {constant=<rgba 1 1 1 0>}, Alpha=@n4.RGB)
  n6 = ParticleColor()
  MaterialFunctionCall(Emissive=Clamp(None=Multiply(A=@n5, B=@n6.RGB)) {maxDefault=100}) {materialFunction=&MF_FX_EmissiveAdjustments}
  n7 = Multiply(A=@n2, B=ScalarParameter() {defaultValue=0.3, parameterName=MaskDistortAmount})
  n8 = TextureCoordinate()
  n9 = OneMinus(None=Distance(A=Add(A=@n7, B=@n8), B=Constant2Vector() {g=0.5, r=0.65}))
  n10 = Saturate(None=Subtract(A=ConstantBiasScale(None=@n9) {bias=-0.5, scale=2}, B=ScalarParameter() {defaultValue=0.1, parameterName=Size}))
  n11 = ScalarParameter() {defaultValue=0.3, parameterName=MaskPow}
  Multiply(A=Multiply(A=@n4.RGB, B=Power(Base=@n10, Exp=@n11)), B=@n6.A)
  Subtract(A=@n3, B=Constant() {r=0.1})
 ]
}
